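; Base lock actor that blocks interaction with an attached door, container or vault until unlocked, with overridable check and event hooks for children.
; On playlist ready it binds to its parent or linked actor; the server finds sibling locks and sets the initial state.
; Interaction changes LockState, whose OnRep drives visuals, opening, loot, quest events and dispatchers, honoring multi-lock checks.
; It optionally relocks on a timer and removes itself when the attached actor dies.
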(event Custom|SetupMultipleLocks
  ;[1 1 1] Setup Multiple Locks - If the lock is part of a set of multiple locks, check other child actors on the parent actor to see if it has any other locks and save out the locks so they can be checked later.
  (if (Variables|LockMultipleLocks|GetHasMultipleLocks)
    (CallFunction|FindOtherLocks)))

(event Custom|FindOtherLocks
  (bind _returnvalue (Actor|GetParentActor))
  ;[1 1 1] Setup Multiple Locks - If the lock is part of a set of multiple locks, check other child actors on the parent actor to see if it has any other locks and save out the locks so they can be checked later.
  (Utilities|IsValid _returnvalue
    (:Is%20Valid
      (for _array_element (Actor|GetComponentsByClass _returnvalue "/Script/Engine.ChildActorComponent")
        (if (or (Utilities|ClassIsChildOf (Variables|/Script/Engine.ChildActorComponent|GetChildActorClass _array_element) "/Game/Athena/Items/EnvironmentalItems/Locks/BGA_Athena_Lock_Parent.BGA_Athena_Lock_Parent_C") (Utilities|Equal(Class) (Variables|/Script/Engine.ChildActorComponent|GetChildActorClass _array_element) "/Game/Athena/Items/EnvironmentalItems/Locks/BGA_Athena_Lock_Parent.BGA_Athena_Lock_Parent_C"))
          (bind _asbga_athena_lock_parent (Utilities|Casting|CastToBGA_Athena_Lock_Parent (Variables|/Script/Engine.ChildActorComponent|GetChildActor _array_element)))
          (Utilities|Array|AddUnique (Variables|LockMultipleLocks|GetOtherLocksOnActor) _asbga_athena_lock_parent))))))

(event Custom|ParentActorDied
  (params (DamagedActor :type Actor%20Object%20Reference) (Damage :type Float%20%28single-precision%29) (InstigatedBy :type Controller%20Object%20Reference) (DamageCauser :type Actor%20Object%20Reference) (HitLocation :type Vector) (FHitComponent :type Primitive%20Component%20Object%20Reference) (BoneName :type Name) (Momentum :type Vector))
  ;[1 1 1] If we have a parent actor and it dies hide and kill the lock.
  (Fortnite|FortShippingLog 0 (Utilities|String|Append "SQUADLOCKS: Parent Actor Died : Damaged Actor = " (Utilities|GetDisplayName DamagedActor) "   Self=" (Utilities|GetDisplayName self)))
  (CallFunction|HideandKill))

(event EventBeginPlay
  ;[1 1 1] Begin Play - On Playlist Data ready set the actor to locked
  (Class|FortAsyncActionCurrentPlaylistReady|PlaylistDataReadyAsync
    (:OnReady
      ;[1 1 1] Check if we should auto attach to parent actors.
      (if (Variables|LockParent|GetAutoAttachToParentActors)
        (bind _asbuilding_actor (Utilities|Casting|CastToBuildingActor (Actor|GetParentActor))
          (:then
            (bind _output_get (Variables|LockParent|SetAttachedLockActor _asbuilding_actor))
            (EventDispatchers|BindEventtoOnDied _output_get (delegate Custom|ParentActorDied)))
          (:CastFailed
            (if (Variables|Default|GetShouldAutoLinkToActorBehindIt)
              (LinkToActorComponent|PerformLinkingCheck (Variables|Default|GetFortLinkToActor) "Backward" (Variables|LockParent|GetBlankLinkedActor)))))
        (if (Networking|IsServer)
          (CallFunction|SetupMultipleLocks)
          ;[1 1 1] Row Checks for enabling and disabling the lock. THIS NEEDS TO TURN OFF TO DISABLE THE WHOLE LOCK. IF THIS IS LEFT ON AND THE ENABLED IS TURNED OFF THE LOCK WILL LOCK THE ACTOR BEFORE DESTROYING ITSELF! (This is intended so that we can lock it and destroy the lock if that is the desired effect)
          (if (not (Class|AthenaFunctionLibrary|RowtoBool (Variables|ParentRows|GetRow_DoNotLockActor)))
            (Variables|LockParent|SetLockState "NewEnumerator1"))
          ;[1 1 1] Check Enabled Row
          (if (not (Class|AthenaFunctionLibrary|RowtoBool (Variables|ParentRows|GetRow_SelfInteractEnabled)))
            (CallFunction|NotEnabled)))))))

(event Custom|NotEnabled)

(event Custom|OnFirstInteract
  (params (InteractingPawn :type Fort%20Pawn%20Object%20Reference)))

(event Interaction|EventBlueprintOnInteract
  (params (InteractingPawn :type Fort%20Pawn%20Object%20Reference) (InteractionBeingAttempted :type EInteractionBeingAttempted%20Enum :default "FirstInteraction"))
  (bind _lockstate (Variables|LockParent|GetLockState))
  (bind _returnvalue (Utilities|Select :Index _lockstate :NewEnumerator0 "NewEnumerator0" :NewEnumerator1 "NewEnumerator2" :NewEnumerator2 "NewEnumerator1"))
  ;[1 1 1] On Interact - Save out the last interacting pawn and then fire off the interact events for children, if required perform lock check on interact(First interact defaults to true) , then unlock if the unlocks are true (Default first interact is true)
  (Variables|LockParentInteract|SetLastInteractingPawn InteractingPawn)
  (switch Utilities|FlowControl|Switch|SwitchonEInteractionBeingAttempted InteractionBeingAttempted
    (:FirstInteraction
      (if (Variables|LockParentInteract|GetDoLockCheckFirstInteract)
        (bind _checkpassed (CallFunction|LockCheck (Variables|LockParentInteract|GetLastInteractingPawn)))
        (if _checkpassed
          _
          (elif (Variables|LockParentInteract|GetApplyEffectOnFailCheck)
            (CallFunction|ApplyFailCheckEffect)))
        (else
          _))
      (if (Variables|LockParentInteract|GetSetLockOnFirstInteract)
        (CallFunction|SetLock _returnvalue)
        (else
          _))
      (CallFunction|OnFirstInteract (Variables|LockParentInteract|GetLastInteractingPawn)))
    (:SecondInteraction
      (if (Variables|LockParentInteract|GetDoLockCheckSecondInteract)
        (bind _checkpassed_2 (CallFunction|LockCheck (Variables|LockParentInteract|GetLastInteractingPawn) "SecondInteraction"))
        (if _checkpassed_2
          _
          (elif (Variables|LockParentInteract|GetApplyEffectOnFailCheck)
            (CallFunction|ApplyFailCheckEffect)))
        (else
          _))
      (if (Variables|LockParentInteract|GetSetLockOnSecondInteract)
        (CallFunction|SetLock _returnvalue)
        (else
          _))
      (CallFunction|OnSecondInteract (Variables|LockParentInteract|GetLastInteractingPawn)))))

(event Custom|OnSecondInteract
  (params (InteractingPawn :type Fort%20Pawn%20Object%20Reference)))

(event Custom|SetLock
  (params (LockState :type Enum_Athena_Lock%20Enum :default "NewEnumerator0"))
  ;[1 1 1] Unlock Actor -  Set the lock state and set ever set lock so that we can check if the lock should continue to have Allow Interact or Not. Also send event to the AttachedLockActor
  (bind _output_get (Variables|LockParent|SetLockState LockState))
  (Variables|LockParentInteract|SetEverInteractSetLock true)
  (Ability|SendGameplayEventtoActor (Variables|LockParent|GetAttachedLockActor) (Utilities|Select :Index _output_get :NewEnumerator1 (Variables|LockParent|GetEvent_Locked) :NewEnumerator2 (Variables|LockParent|GetEvent_Unlocked)))
  (CallFunction|ExecuteValidGC (Utilities|Select :Index _output_get :NewEnumerator1 (Variables|LockGameplayCues|GetGC_Locked) :NewEnumerator2 (Variables|LockGameplayCues|GetGC_Unlocked)))
  ;[1 1 1] Automatically Relock after delay if option is available
  (if (and (Variables|LockParent|GetAutoRelock) (Utilities|Enum|Equal(Enum) _output_get "NewEnumerator2"))
    (Utilities|Time|SetTimerbyEvent (delegate Custom|TimedRelock) (ScalableFloat|GetValueatLevel (Variables|LockParent|GetRow_RelockDelay)))))

(event Custom|UnlockOpenDoor
  (bind _attachedlockactor (Variables|LockParent|GetAttachedLockActor))
  (bind _returnvalue (GameplayTags|GetOwnedGameplayTags _attachedlockactor :!class /Script/GameplayTags.GameplayTagAssetInterface))
  ;[1 1 1] Auto Open Door On Unlock
  (if (Variables|LockParent|GetAutoDoorOnUnlock)
    (bind _returnvalue_2 (Openable|OpenActor (Variables|LockParent|GetAttachedLockActor) (Fortnite|GetFortPlayerControllerfromActor (Variables|LockParentInteract|GetLastInteractingPawn))))
    ;[1 1 1] Update Player Quest
    (if _returnvalue_2
      (bind _lastinteractingpawn (Variables|LockParentInteract|GetLastInteractingPawn)
        (:then
          (Utilities|IsValid (Fortnite|GetFortPlayerControllerfromActor _lastinteractingpawn)
            (:Is%20Valid
              (Utilities|IsValid (Quests|GetQuestManager (Fortnite|GetFortPlayerControllerfromActor _lastinteractingpawn) "Invalid")
                (:Is%20Valid
                  (Quest|SendComplexCustomStatEvent (Quests|GetQuestManager (Fortnite|GetFortPlayerControllerfromActor _lastinteractingpawn) "Invalid") 0 "(GameplayTags=((TagName=\"Athena.Quests.UnlockFaction.Door\")))" _returnvalue 1)
                  (bind _result (CustomVerb|AddTagsforSubject (CustomVerb|MakeCustomVerbMessage (Fortnite|GetFortPlayerControllerfromActor _lastinteractingpawn) "(TagName=\"CustomVerb.KeyUnlocked\")") "UnlockedActor" _returnvalue))
                  (CustomVerb|BroadcastCustomVerbMessage _result))))))
        (:else)))))

(event Custom|TimedRelock
  ;[1 1 1] Automatically Relock after delay if option is available
  (CallFunction|SetLock "NewEnumerator1"))

(event Custom|LockCloseDoor
  ;[1 1 1] Auto Close Door On Lock
  (if (Variables|LockParent|GetAutoCloseDoorOnLock)
    (Openable|CloseActor (Variables|LockParent|GetAttachedLockActor))))

(event Custom|ApplyFailCheckEffect
  (bind _ge_failedcheck (Variables|LockParentInteract|GetGE_FailedCheck))
  (bind _lastinteractingpawn (Variables|LockParentInteract|GetLastInteractingPawn))
  ;[1 1 1] Apply Fail Check Effect - If valid apply gameplay effect to the last interacting pawn
  (Utilities|IsValid _lastinteractingpawn
    (:Is%20Valid
      (if (Utilities|IsValidClass _ge_failedcheck)
        (GameplayEffects|ApplyGameplayEffectToTarget :self (Ability|GetFortAbilitySystemComponent self) :GameplayEffectClass _ge_failedcheck :Target (Ability|GetFortAbilitySystemComponent _lastinteractingpawn) :Level 0.0)))))

(event OnLinkedActorDestroyed(FortLinkToActor)
  (params (DamageCauser :type Actor%20Object%20Reference))
  ;[1 1 1] On Linked Actor Destroyed - Hide and Kill
  (Fortnite|FortShippingLog 0 (Utilities|String|Append "SQUADLOCKS: On Linked Actor Destroyed   Damage Causer=" (Utilities|GetDisplayName DamageCauser) "   Self=" (Utilities|GetDisplayName self)))
  (CallFunction|HideandKill))

(event Custom|OpenContainers
  (bind _lastinteractingpawn (Variables|LockParentInteract|GetLastInteractingPawn))
  (bind _returnvalue (Fortnite|GetFortPlayerControllerfromActor _lastinteractingpawn))
  (bind _returnvalue_2 (GameplayTags|GetOwnedGameplayTags (Variables|LockParent|GetAttachedLockActor) :!class /Script/GameplayTags.GameplayTagAssetInterface))
  (bind _attachedlockactor (Variables|LockParent|GetAttachedLockActor))
  ;[1 1 1] Auto Open containers
  (if (Variables|LockParent|GetAutoOpenContainersOnUnlock)
    (bind _asbuilding_container (Utilities|Casting|CastToBuildingContainer _attachedlockactor))
    (bind _asfort_player_pawn (Utilities|Casting|CastToFortPlayerPawn (Variables|LockParentInteract|GetLastInteractingPawn)))
    (bind _returnvalue_3 (Container|SpawnLoot _asbuilding_container _asfort_player_pawn))
    ;[1 1 1] Update Player Quest
    (if _returnvalue_3
      (Utilities|IsValid _returnvalue
        (:Is%20Valid
          (Quest|SendComplexCustomStatEvent (Quests|GetQuestManager _returnvalue "Invalid") 0 "(GameplayTags=((TagName=\"Athena.Quests.UnlockFaction.Chest\")))" _returnvalue_2 1)
          (bind _result (CustomVerb|AddTagsforSubject (CustomVerb|MakeCustomVerbMessage _returnvalue "(TagName=\"CustomVerb.KeyUnlocked\")") "UnlockedActor" _returnvalue_2))
          (CustomVerb|BroadcastCustomVerbMessage _result))))))

(event Custom|OnLocked
  (CallFunction|LockCloseDoor)
  (CallFunction|SetLockVisuals))

(event Custom|HideAndKill
  ;[1 1 1] Hide And Kill - Set hidden, And set LifeSpan
  (Fortnite|FortShippingLog 0 "SQUADLOCKS: Hide And Kill")
  (Rendering|SetActorHiddenInGame true)
  (Variables|Interaction|SetbAllowInteract false)
  (Actor|SetLifeSpan 0.1))

(event Custom|OnUnlocked
  (Utilities|CheckMultiLocks
    (:true
      (Utilities|FlowControl|Sequence
        (:then_0
          (CallFunction|OpenContainers)
          (CallFunction|UnlockOpenDoor)
          (CallFunction|SetLockVisuals)
          (Utilities|Casting|CastToVault_Door (Variables|LockParent|GetAttachedLockActor))
          (Utilities|Time|SetTimerbyEvent (delegate Custom|CallOpenVault) 1.0))
        (:then_1
          (if (Variables|LockMultipleLocks|GetHasMultipleLocks)
            (bind _array_element (Utilities|Array|ForEachLoopwithBreak (Variables|LockMultipleLocks|GetOtherLocksOnActor))
              (:LoopBody
                (CallFunction|MultilockCompleted _array_element)))
            (else
              (CallFunction|MultilockCompleted))))))))

(event Custom|CallOpenVault
  (bind _asvault_door (Utilities|Casting|CastToVault_Door (Variables|LockParent|GetAttachedLockActor)))
  (Class|VaultDoor|VaultOpen _asvault_door))

(event EventEndPlay
  (params (EndPlayReason :type EEndPlayReason%20Enum :default "Destroyed"))
  (Fortnite|FortShippingLog 0 (Utilities|String|Append "SQUADLOCKS: End play for :" (Utilities|GetObjectPathString self) "    Reason: " (Utilities|String|EnumtoString EndPlayReason)))
  (|CallParentFunction EndPlayReason))

(event Custom|MultilockCompleted)

(fn AllLocksUnlocked
  (LocalVariables|SetAreAnyLocked false)
  ;[1 1 1] Check if any of the multi locks are locked.
  (bind _array_element (Utilities|Array|ForEachLoopwithBreak (Variables|LockMultipleLocks|GetOtherLocksOnActor))
    (:LoopBody
      (if (Utilities|Enum|Equal(Enum) (Variables|LockParent|GetLockState _array_element) "NewEnumerator1")
        (LocalVariables|SetAreAnyLocked true)))
    (:Completed
      (return (not (LocalVariables|GetAreAnyLocked))))))

(fn BlueprintCanInteract
  (params (InteractingPawn :type Fort%20Pawn%20Object%20Reference) (InteractionBeingAttempted :type EInteractionBeingAttempted%20Enum :default "FirstInteraction") (InteractionType :type TInteractionType%20Enum :default "IT_NoInteraction"))
  (Utilities|IsValid InteractingPawn
    (:Is%20Valid
      ;[1 1 1] Check Interact Angle
      (if (<= (Math|Trig|Acos(Degrees) (Math|Vector|DotProduct (Transformation|GetActorForwardVector) (Math|Vector|GetForwardVector (Math|Rotator|MakeRotator 0.0 0.0 (.yaw (Math|Rotator|FindLookatRotation (Transformation|GetActorLocation) (Transformation|GetActorLocation InteractingPawn))))))) (Variables|LockParentInteract|GetMaxInteractAngle))
        ;[1 1 1] Check If the lock can be used while unlocked.
        (if (Variables|LockParentInteract|GetCanInteractWhileUnlocked)
          _
          (elif (Utilities|Enum|Equal(Enum) (Variables|LockParent|GetLockState) "NewEnumerator1")
            _
            (else
              (return false))))
        (return (Utilities|Select :Index InteractionBeingAttempted :FirstInteraction (Variables|LockParentInteract|GetCanFirstInteract) :SecondInteraction (Variables|LockParentInteract|GetCanSecondInteract) :AllInteraction false))
        (else
          (return false))))
    (:Is%20Not%20Valid
      (return false))) ; If not valid pawn return false
)

(fn BlueprintGetInteractionString
  (params (InteractingPawn :type Fort%20Pawn%20Object%20Reference) (InteractionBeingAttempted :type EInteractionBeingAttempted%20Enum :default "FirstInteraction"))
  (return (Utilities|Select :Index InteractionBeingAttempted :FirstInteraction (Variables|LockParentInteract|GetFirstInteractText) :SecondInteraction (Variables|LockParentInteract|GetSecondInteractText))))

(fn BlueprintGetInteractionTime
  (params (InteractingPawn :type Fort%20Pawn%20Object%20Reference) (InteractionBeingAttempted :type EInteractionBeingAttempted%20Enum :default "FirstInteraction"))
  (return (Utilities|Select :Index InteractionBeingAttempted :FirstInteraction (ScalableFloat|GetValueatLevel (Variables|LockParentInteract|GetFirstInteractTime)) :SecondInteraction (ScalableFloat|GetValueatLevel (Variables|LockParentInteract|GetSecondInteractTime)) :AllInteraction 0.0) true))

(fn ExecuteValidGC
  (params (GameplayCueTag :type Gameplay%20Tag%20Structure))
  ;[1 1 1] Check if tag is valid if it is execute the GC
  (if (GameplayTags|IsGameplayTagValid GameplayCueTag)
    (Ability|ExecuteGameplayCueReplicated GameplayCueTag)))

(fn LockSetInteractCollision
  (params (On :type Boolean))
  ;[1 1 1] Turn On and off collision for the interact
  (if On
    (Collision|SetCollisionEnabled (Variables|Default|GetInteractionCollision) "QueryAndPhysics")
    (else
      (Collision|SetCollisionEnabled (Variables|Default|GetInteractionCollision) "NoCollision"))))

(fn LockCheck
  (params (InteractingPawn :type Fort%20Pawn%20Object%20Reference) (InteractionBeingAttempted :type EInteractionBeingAttempted%20Enum :default "FirstInteraction"))
  (return true))

(fn OnRep_EverInteractSetLock
  ;[1 1 1] On Rep - If we cannot interact multiple times then set Allow Interact to false
  (if (not (Variables|LockParent|GetCanMultipleInteract))
    (Variables|Interaction|SetbAllowInteract false)))

(fn OnRep_LockState
  (bind _attachedlockactor (Variables|LockParent|GetAttachedLockActor))
  (switch Utilities|FlowControl|Switch|SwitchonEnum_Athena_Lock (Variables|LockParent|GetLockState)
    (:NewEnumerator1
      (Utilities|IsValid _attachedlockactor
        (:Is%20Valid
          (Variables|/Script/FortniteGame.BuildingActor|SetbAllowInteract false _attachedlockactor)
          ;[1 1 1] Send out events that can be overridden
          (CallFunction|OnLocked)
          ;[1 1 1] Send out dispatchers so that they can be listened for.
          (Default|CallOnLockedDispatcher))))
    (:NewEnumerator2
      (Utilities|IsValid _attachedlockactor
        (:Is%20Valid
          (if (Variables|LockParentInteract|GetSetsAllowInteractOnUnlock)
            (Utilities|CheckMultiLocks
              (:true
                (Variables|/Script/FortniteGame.BuildingActor|SetbAllowInteract true _attachedlockactor))
              (:false
                _))
            (CallFunction|OnUnlocked)
            (Default|CallOnUnlockedDispatcher)
            (else
              _)))))))

(fn SetLockVisuals
  (bind _lockstate (Variables|LockParent|GetLockState))
  (bind _unlockedmesh (Variables|LockCosmetics|GetUnlockedMesh))
  (bind _lockedmesh (Variables|LockCosmetics|GetLockedMesh))
  (bind _returnvalue (Utilities|Select :Index _lockstate :NewEnumerator1 _lockedmesh :NewEnumerator2 _unlockedmesh))
  (bind _lockstate_2 (Variables|LockParent|GetLockState))
  (bind _unlockedmaterial (Variables|LockCosmetics|GetUnlockedMaterial))
  (bind _lockedmaterial (Variables|LockCosmetics|GetLockedMaterial))
  (bind _returnvalue_2 (Utilities|Select :Index _lockstate_2 :NewEnumerator1 _lockedmaterial :NewEnumerator2 _unlockedmaterial))
  (Networking|CanExecuteCosmeticEvents
    (:True
      (Utilities|IsValid _returnvalue
        (:Is%20Valid
          (Components|StaticMesh|SetStaticMesh (Variables|Default|GetLock) _returnvalue))
        (:Is%20Not%20Valid
          _))
      (Utilities|IsValid _returnvalue_2
        (:Is%20Valid
          (Rendering|Material|SetMaterial (Variables|Default|GetLock) 0 _returnvalue_2))))))

(fn ShouldDie
  (params (Damage :type Float%20%28single-precision%29) (EventInstigator :type Controller%20Object%20Reference) (DamageCauser :type Actor%20Object%20Reference) (bIsTest :type Boolean) (EffectContext :type Gameplay%20Effect%20Context%20Handle%20Structure%20%28by%20ref%29))
  (return false))

(fn ConstructionScript)
